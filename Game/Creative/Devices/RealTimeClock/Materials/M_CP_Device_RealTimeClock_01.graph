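# Default Lit material for the Creative Real Time Clock device, combining textured body surfaces with a stylized emissive screen and logo.
# Screen image on UV1 takes tint, desaturation and switchable text, then animated sine scanlines.
# A contrast-adjusted UV1 vertical gradient masks the screen into diffuse and glow, with logo and halo layered from R/G masks.
# Specular mask blue drives a min-max blend; separate pulse and camera-distance-scaled height calculations also exist.
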Material /Game/Creative/Devices/RealTimeClock/Materials/M_CP_Device_RealTimeClock_01
graph {
 properties {pixelDepthOffsetMode=PDOM_Legacy, usedShadingModels=Default Lit}
 nodes [
  n1 = ScalarParameter() {defaultValue=20000, parameterName=ScaleDistance}
  n2 = Divide(A=Distance(A=CameraPositionWS(), B=ActorPositionWS()), B=@n1) {constB=1000}
  Divide(A=ComponentMask(None=WorldPosition().XYZ) {b=true}, B=LinearInterpolate(B=@n1, Alpha=Saturate(None=@n2)) {constA=50, constB=100}) {constB=20}
  LinearInterpolate(
   A=ScalarParameter() {group=Specular, parameterName=SpecRoughnessMin}
   B=ScalarParameter() {defaultValue=0.9, group=Specular, parameterName=SpecRoughnessMax}
   Alpha=TextureSampleParameter2D() {parameterName=SpecularMasks, texture=/Game/Creative/Devices/RealTimeClock/Textures/T_CP_Device_RealTimeClock_01_S}.B
  )
  n3 = ComponentMask(None=TextureCoordinate()) {g=true}
  n4 = Multiply(A=Time(), B=ScalarParameter() {defaultValue=0.02, group=Screen Image, parameterName=ScanlinePanSpeed})
  n5 = Sine(None=Multiply(A=ScalarParameter() {defaultValue=80, group=Screen Image, parameterName=ScanlineTile}, B=Add(A=@n3, B=@n4)) {constA=1})
  n6 = AppendVector(
   A=ScalarParameter() {defaultValue=1, group=Screen Image, parameterName=Screen Image Scale U, sortPriority=51}
   B=ScalarParameter() {defaultValue=1, group=Screen Image, parameterName=Screen Image Scale V, sortPriority=51}
  )
  n7 = TextureCoordinate() {coordinateIndex=1, desc=2nd UV Set}
  n8 = ScalarParameter() {group=Screen Image, parameterName=Screen Image Offset U, sortPriority=51}
  n9 = ScalarParameter() {group=Screen Image, parameterName=Screen Image Offset V, sortPriority=51}
  n10 = TextureSampleParameter2D(UVs=Add(
    A=MaterialFunctionCall(UVs=@n7, Texture Scale=@n6) {materialFunction=&ScaleUVsByCenter}.UVs
    B=AppendVector(A=@n8, B=@n9)
   )) {
   group=Screen Image
   parameterName=Screen Image Texture
   sortPriority=50
   texture=/Game/Creative/Devices/StoreDevice/Textures/T_CP_Device_StoreDevice_ImageDefault_01_D
  }
  n11 = Multiply(
   A=Desaturation(None=@n10.RGB, Fraction=ScalarParameter() {group=Screen Image, parameterName=Desaturation, sliderMax=1, sortPriority=55})
   B=VectorParameter() {group=Screen Image, parameterName=Tint, sortPriority=55, defaultValue=<rgba 1 1 1 1>}.RGB
  )
  n12 = AppendVector(
   A=ScalarParameter() {defaultValue=1, group=Screen Image, parameterName=Text Scale U, sortPriority=42}
   B=ScalarParameter() {defaultValue=1, group=Screen Image, parameterName=Text Scale V, sortPriority=42}
  )
  n13 = Constant()
  n14 = ScalarParameter() {group=Screen Image, parameterName=Text Offset V, sortPriority=42}
  n15 = TextureSampleParameter2D(UVs=Add(
    A=MaterialFunctionCall(UVs=@n7, Texture Scale=@n12) {materialFunction=&ScaleUVsByCenter}.UVs
    B=AppendVector(A=@n13, B=@n14)
   )) {
   group=Screen Image
   parameterName=Text Texture
   samplerSource=SSM_Clamp_WorldGroupSettings
   sortPriority=41
   texture=/Game/Creative/Devices/StoreDevice/Textures/T_CP_Device_StoreDevice_TextDefault_01_D
  }
  n16 = StaticSwitchParameter(
   True=LinearInterpolate(A=@n11, B=VectorParameter() {group=Screen Image, parameterName=Text Color, sortPriority=42, defaultValue=<rgba 1 1 1 1>}.RGB, Alpha=@n15.R)
   False=@n11
  ) {group=Screen Image, parameterName=Use Screen Text, sortPriority=40}
  n17 = LinearInterpolate(
   B=Multiply(A=ConstantBiasScale(None=@n5), B=@n16)
   Alpha=ScalarParameter() {defaultValue=0.6, group=Screen Image, parameterName=ScanlineIntensity}
  ) {constA=1}
  n18 = Multiply(A=@n16, B=ScalarParameter() {defaultValue=1, group=Screen Image, parameterName=Emissive Multiplier, sortPriority=52})
  n19 = MaterialFunctionCall(Float2=@n7) {materialFunction=&BreakOutFloat2Components}
  n20 = MaterialFunctionCall(
   In=Power(Base=OneMinus(None=@n19.G), Exp=ScalarParameter() {defaultValue=0.1, group=Screen Image, parameterName=Mask Exponent, sortPriority=10})
   Contrast=ScalarParameter() {defaultValue=1, group=Screen Image, parameterName=Mask Contrast, sortPriority=11}
  ) {materialFunction=&CheapContrast}
  n21 = TextureCoordinate() {coordinateIndex=1, desc=Second UV Set}
  n22 = ScalarParameter() {defaultValue=1, group=Logo, parameterName=Logo Scale}
  n23 = ScalarParameter() {group=Logo, parameterName=Logo Offset U}
  n24 = ScalarParameter() {group=Logo, parameterName=Logo Offset V}
  n25 = TextureSampleParameter2D(UVs=Add(
    A=MaterialFunctionCall(UVs=@n21, Texture Scale=@n22) {materialFunction=&ScaleUVsByCenter}.UVs
    B=AppendVector(A=@n23, B=@n24)
   )) {
   group=Logo
   parameterName=Logo Mask
   samplerSource=SSM_Clamp_WorldGroupSettings
   texture=/Game/Creative/Devices/StoreDevice/Textures/T_CP_Device_StoreDevice_Logo_01_D
  }
  n26 = ScalarParameter() {defaultValue=0.2, parameterName=Glow Intensity}
  LinearInterpolate(
   A=LinearInterpolate(A=Constant(), B=Multiply(A=@n17, B=@n18), Alpha=@n20.Result) {constAlpha=1}
   B=LinearInterpolate(
    A=VectorParameter() {group=Logo, parameterName=Glow Color, sortPriority=1, defaultValue=<rgba 0 1 0.9387 1>}.RGB
    B=VectorParameter() {group=Logo, parameterName=Logo Color, sortPriority=1, defaultValue=<rgba 0.9047 1 1 1>}.RGB
    Alpha=@n25.R
   ) {constAlpha=1}
   Alpha=LinearInterpolate(A=Multiply(A=@n25.G, B=@n26), B=@n25.R, Alpha=@n25.R) {constAlpha=1}
  ) {constAlpha=1}
  LinearInterpolate(
   A=TextureSampleParameter2D() {parameterName=Diffuse, texture=/Game/Creative/Devices/RealTimeClock/Textures/T_CP_Device_RealTimeClock_01_D}.RGB
   B=@n16
   Alpha=@n20.Result
  ) {constAlpha=1}
  MaterialFunctionCall(Value=Time(), Period=ScalarParameter() {defaultValue=2, parameterName=Pulse Speed}) {materialFunction=&LinearSine}
  TextureSampleParameter2D() {parameterName=Normals, samplerType=SAMPLERTYPE_Normal, texture=/Game/Creative/Devices/RealTimeClock/Textures/T_CP_Device_RealTimeClock_01_N}
 ]
}
