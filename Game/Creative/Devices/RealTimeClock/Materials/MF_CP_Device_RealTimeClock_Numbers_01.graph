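# Material function for the RealTimeClock device that draws a scalar input as a two- or three-digit number from a digit-strip texture.
# Floor and Frac math turns each digit into a U offset on the strip, with inputs for V scale and offset.
# SphereMasks on tiled U confine each sample to its slot, and summed R and G channels become separate number and glow masks per layout.
# If nodes comparing against 99 and several scalar parameters feed no output.
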MaterialFunction /Game/Creative/Devices/RealTimeClock/Materials/MF_CP_Device_RealTimeClock_Numbers_01
graph {
 properties {description=2 or 3 Digit Number}
 nodes [
  n1 = FunctionInput(Preview=Constant() {r=123}) {
   bUsePreviewValueAsDefault=true
   desc=2 or 3 Digit Number
   description="Default = 00"
   inputName=2 or 3 Digit Number
   inputType=FunctionInput_Scalar
  }
  n2 = Multiply(A=Floor(None=Multiply(A=@n1) {constB=0.1})) {constB=0.1}
  n3 = Constant()
  n4 = TextureCoordinate() {uTiling=3}
  n5 = FunctionInput(Preview=Constant() {r=0.25}) {bUsePreviewValueAsDefault=true, desc=V Scale, description="Default = 0.6", inputName=V Scale, inputType=FunctionInput_Scalar}
  n6 = AppendVector(A=Constant() {r=1}, B=@n5)
  n7 = Constant2Vector() {g=1, r=0.068}
  n8 = FunctionInput(Preview=Constant() {r=0.5}) {bUsePreviewValueAsDefault=true, desc=V Offset, description="Default = 0", inputName=V Offset, inputType=FunctionInput_Scalar}
  n9 = Add(
   A=Multiply(
    A=MaterialFunctionCall(UVs=@n4, Texture Scale=@n6) {materialFunction=&ScaleUVsByCenter}.UVs
    B=@n7
   )
   B=AppendVector(A=Constant(), B=@n8)
  )
  n10 = FunctionInput(Preview=TextureObject() {texture=/Game/Creative/Devices/RealTimeClock/Textures/T_CP_Device_RealTimeClock_Numbers_01_D}) {bUsePreviewValueAsDefault=true, desc=Number Texture, inputName=Number Texture, inputType=FunctionInput_Texture2D}
  n11 = TextureSample(UVs=Add(A=@n9, B=AppendVector(A=@n2, B=@n3)), Tex=@n10) {samplerType=SAMPLERTYPE_LinearColor}
  n12 = SphereMask(
   A=MaterialFunctionCall(Float2=TextureCoordinate() {uTiling=3}) {materialFunction=&BreakOutFloat2Components}.R
   B=Constant() {r=0.5}
   Radius=Constant() {r=1}
   Hardness=Constant() {r=1}
  )
  n13 = Constant2Vector() {r=-0.1}
  n14 = Frac(None=Multiply(A=@n1) {constB=0.1})
  n15 = Constant()
  n16 = TextureSample(UVs=Add(A=Add(A=@n9, B=@n13) {constB=-0.1}, B=AppendVector(A=@n14, B=@n15)), Tex=@n10) {samplerType=SAMPLERTYPE_LinearColor}
  n17 = SphereMask(
   A=MaterialFunctionCall(Float2=TextureCoordinate() {uTiling=3}) {materialFunction=&BreakOutFloat2Components}.R
   B=Constant() {r=2.5}
   Radius=Constant() {r=1}
   Hardness=Constant() {r=1}
  )
  n18 = Add(A=Multiply(A=@n11.R, B=@n12), B=Multiply(A=@n16.R, B=@n17))
  FunctionOutput(None=@n18) {outputName=2 Digit Number Mask, sortPriority=1}
  n19 = Add(A=Multiply(A=@n11.G, B=@n12), B=Multiply(A=@n16.G, B=@n17))
  FunctionOutput(None=@n19) {outputName=2 Digit Glow Mask, sortPriority=2}
  n20 = Multiply(A=Floor(None=Multiply(A=@n1) {constB=0.01})) {constB=0.1}
  n21 = Constant()
  n22 = TextureCoordinate() {uTiling=3}
  n23 = AppendVector(A=Constant() {r=1}, B=@n5)
  n24 = Add(
   A=Multiply(
    A=MaterialFunctionCall(UVs=@n22, Texture Scale=@n23) {materialFunction=&ScaleUVsByCenter}.UVs
    B=Constant2Vector() {g=1, r=0.1}
   )
   B=AppendVector(A=Constant(), B=@n8)
  )
  n25 = TextureSample(UVs=Add(A=@n24, B=AppendVector(A=@n20, B=@n21)), Tex=@n10) {samplerType=SAMPLERTYPE_LinearColor}
  n26 = SphereMask(
   A=MaterialFunctionCall(Float2=TextureCoordinate() {uTiling=3}) {materialFunction=&BreakOutFloat2Components}.R
   B=Constant() {r=0.5}
   Radius=Constant() {r=0.5}
   Hardness=Constant() {r=1}
  )
  n27 = Constant2Vector() {r=-0.1}
  n28 = Constant()
  n29 = TextureSample(UVs=Add(A=Add(A=@n24, B=@n27) {constB=-0.1}, B=AppendVector(A=@n2, B=@n28)), Tex=@n10) {samplerType=SAMPLERTYPE_LinearColor}
  n30 = SphereMask(
   A=MaterialFunctionCall(Float2=TextureCoordinate() {uTiling=3}) {materialFunction=&BreakOutFloat2Components}.R
   B=Constant() {r=1.5}
   Radius=Constant() {r=0.5}
   Hardness=Constant() {r=1}
  )
  n31 = Constant2Vector() {r=-0.2}
  n32 = Constant()
  n33 = TextureSample(UVs=Add(A=Add(A=@n24, B=@n31) {constB=-0.1}, B=AppendVector(A=@n14, B=@n32)), Tex=@n10) {samplerType=SAMPLERTYPE_LinearColor}
  n34 = SphereMask(
   A=MaterialFunctionCall(Float2=TextureCoordinate() {uTiling=3}) {materialFunction=&BreakOutFloat2Components}.R
   B=Constant() {r=2.5}
   Radius=Constant() {r=0.5}
   Hardness=Constant() {r=1}
  )
  n35 = Add(A=Add(A=Multiply(A=@n25.R, B=@n26), B=Multiply(A=@n29.R, B=@n30)), B=Multiply(A=@n33.R, B=@n34))
  FunctionOutput(None=@n35) {outputName=3 Digit Number Mask, sortPriority=3}
  n36 = Add(A=Add(A=Multiply(A=@n25.G, B=@n26), B=Multiply(A=@n29.G, B=@n30)), B=Multiply(A=@n33.G, B=@n34))
  FunctionOutput(None=@n36) {outputName=3 Digit Glow Mask, sortPriority=4}
  n37 = Constant2Vector() {g=1, r=0.1}
  If(A=@n1, B=Constant() {r=99}, "A > B"=@n37, "A == B"=@n7, "A < B"=@n7)
  If(A=@n1, B=Constant() {r=99}, "A > B"=@n36, "A == B"=@n19, "A < B"=@n19)
  If(A=@n1, B=Constant() {r=99}, "A > B"=@n35, "A == B"=@n18, "A < B"=@n18)
  Reroute(None=@n37)
  ScalarParameter() {parameterName=Tenth}
  ScalarParameter() {parameterName=Single}
  ScalarParameter() {parameterName=Tenth}
  ScalarParameter() {parameterName=Single}
  ScalarParameter() {parameterName=Hundredth}
 ]
}
