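# Material function for the Creative real-time clock device that converts an end time in client game seconds into remaining Days, Hours, Minutes and Seconds.
# It subtracts the floored, pause-ignoring material Time from EndTime and clamps the difference at zero.
# That remainder is divided into each unit, wrapped with Fmod for hours, minutes and seconds, and truncated to whole numbers on four separate outputs.
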MaterialFunction /Game/Creative/Devices/RealTimeClock/Materials/MF_RemainingTimeFromEnd
graph {
 properties {description="From an input end time in seconds, returns the days, hours, minutes and seconds till that time"}
 nodes [
  n1 = FunctionInput() {
   description="The end time for your timer, as time from client game seconds."
   inputName=EndTime
   inputType=FunctionInput_Scalar
   previewValue=<100 0 0 1>
  }
  n2 = Time() {bIgnorePause=true}
  n3 = Max(A=Subtract(A=@n1, B=Floor(None=@n2)) {constB=0}) {constB=0}
  n4 = Divide(A=@n3) {constB=3600, desc="seconds -> hours"}
  n5 = Constant() {r=24}
  FunctionOutput(None=Truncate!(None=Fmod(A=@n4, B=@n5)) {is=class MaterialExpressionTruncate}) {outputName=Hours, sortPriority=1}
  FunctionOutput(None=Truncate(None=Divide(A=@n3) {constB=86400, desc="seconds -> days"})) {outputName=Days}
  n6 = Divide(A=@n3) {constB=60, desc="seconds -> mins"}
  n7 = Constant() {r=60}
  FunctionOutput(None=Truncate(None=Fmod(A=@n6, B=@n7))) {outputName=Minutes, sortPriority=2}
  n8 = Constant() {r=60}
  FunctionOutput(None=Truncate(None=Fmod(A=@n3, B=@n8))) {outputName=Seconds, sortPriority=3}
 ]
}
