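# Unlit translucent material that draws the Creative RealTimeClock device's days, hours, minutes and seconds as glowing digits.
# Each value comes from a parameter, or from MF_RemainingTimeFromEnd when EndTime is positive, and feeds a shared digit-texture function.
# SphereMasks on UV V place the four fields, driving color and opacity, both scaled by Enabled.
# The Is Hologram switch drops glow from opacity and spins the mesh around Z via RotateAboutAxis.
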Material /Game/Creative/Devices/RealTimeClock/Materials/M_CP_Device_RealTimeClock_Numbers_01
graph {
 properties {blendMode=BLEND_Translucent, pixelDepthOffsetMode=PDOM_Legacy, shadingModel=MSM_Unlit, usedShadingModels=Unlit}
 nodes [
  n1 = ScalarParameter() {parameterName=Number Day}
  n2 = ScalarParameter() {defaultValue=-1, desc="if > 0, calculate time\r\nremaining in material", parameterName=EndTime}
  n3 = MaterialFunctionCall!(EndTime=@n2) {materialFunction=&/MF_RemainingTimeFromEnd}
  n4 = ScalarParameter() {defaultValue=0.25, parameterName=V Scale}
  n5 = ScalarParameter() {defaultValue=0.5, parameterName=V Offset}
  n6 = TextureObjectParameter() {parameterName=Numbers Texture, texture=/Game/Creative/Devices/RealTimeClock/Textures/T_CP_Device_RealTimeClock_Numbers_01_D}
  n7 = MaterialFunctionCall!(2 or 3 Digit Number=If(A=@n2, "A > B"=@n3.Days, "A == B"=@n1, "A < B"=@n1), V Scale=@n4, V Offset=@n5, Number Texture=@n6) {materialFunction=&/MF_CP_Device_RealTimeClock_Numbers_01}
  n8 = MaterialFunctionCall(Float2=TextureCoordinate() {uTiling=3}) {materialFunction=&BreakOutFloat2Components}
  n9 = SphereMask(A=@n8.G, B=Constant(), Radius=Constant() {r=0.25}, Hardness=Constant() {r=1})
  n10 = ScalarParameter() {parameterName=Number Hrs}
  n11 = Constant() {r=1.5}
  n12 = MaterialFunctionCall(2 or 3 Digit Number=If(A=@n2, "A > B"=@n3.Hours, "A == B"=@n10, "A < B"=@n10), V Scale=@n4, V Offset=@n11, Number Texture=@n6) {materialFunction=&MF_CP_Device_RealTimeClock_Numbers_01}
  n13 = SphereMask(A=@n8.G, B=Constant() {r=0.375}, Radius=Constant() {r=0.125}, Hardness=Constant() {r=1})
  n14 = ScalarParameter() {parameterName=Number Min}
  n15 = MaterialFunctionCall(2 or 3 Digit Number=If(A=@n2, "A > B"=@n3.Minutes, "A == B"=@n14, "A < B"=@n14), V Scale=@n4, V Offset=@n5, Number Texture=@n6) {materialFunction=&MF_CP_Device_RealTimeClock_Numbers_01}
  n16 = SphereMask(A=@n8.G, B=Constant() {r=0.625}, Radius=Constant() {r=0.125}, Hardness=Constant() {r=1})
  n17 = Add(
   A=Add(A=Multiply(A=@n7.3 Digit Number Mask, B=@n9), B=Multiply(A=@n12.2 Digit Number Mask, B=@n13))
   B=Multiply(A=@n15.2 Digit Number Mask, B=@n16)
  )
  n18 = ScalarParameter() {parameterName=Number Sec}
  n19 = MaterialFunctionCall(2 or 3 Digit Number=If(A=@n2, "A > B"=@n3.Seconds, "A == B"=@n18, "A < B"=@n18), V Scale=@n4, V Offset=@n11, Number Texture=@n6) {materialFunction=&MF_CP_Device_RealTimeClock_Numbers_01}
  n20 = SphereMask(A=@n8.G, B=Constant() {r=1}, Radius=Constant() {r=0.25}, Hardness=Constant() {r=1})
  n21 = Multiply(A=@n19.2 Digit Number Mask, B=@n20)
  n22 = ScalarParameter() {defaultValue=1, parameterName=Enabled}
  LinearInterpolate(
   B=LinearInterpolate(
    A=VectorParameter() {parameterName=Glow Color, sortPriority=1, defaultValue=<rgba 0 1 0.9387 1>}.RGB
    B=VectorParameter() {parameterName=Color, sortPriority=1, defaultValue=<rgba 0.9047 1 1 1>}.RGB
    Alpha=Add(A=@n17, B=@n21)
   ) {constAlpha=1}
   Alpha=@n22
  ) {constAlpha=1}
  n23 = LinearInterpolate(A=@n7.3 Digit Glow Mask, B=Constant() {r=1}, Alpha=@n7.3 Digit Glow Mask) {constAlpha=0}
  n24 = LinearInterpolate(A=@n12.2 Digit Glow Mask, B=Constant() {r=1}, Alpha=@n12.2 Digit Glow Mask) {constAlpha=0}
  n25 = Add(
   A=Multiply(A=StaticSwitchParameter(True=@n7.3 Digit Number Mask, False=@n23) {parameterName=Is Hologram}, B=@n9)
   B=Multiply(A=StaticSwitchParameter(True=@n12.2 Digit Number Mask, False=@n24) {parameterName=Is Hologram}, B=@n13)
  )
  n26 = Constant() {r=1}
  n27 = Multiply(
   A=StaticSwitchParameter(True=@n15.2 Digit Number Mask, False=LinearInterpolate(A=@n15.2 Digit Glow Mask, B=@n26, Alpha=@n15.2 Digit Glow Mask) {constAlpha=0}) {parameterName=Is Hologram}
   B=@n16
  )
  n28 = StaticSwitchParameter(
   True=@n19.2 Digit Number Mask
   False=LinearInterpolate(A=@n19.2 Digit Glow Mask, B=Constant() {r=1}, Alpha=@n19.2 Digit Glow Mask) {constAlpha=0}
  ) {parameterName=Is Hologram}
  LinearInterpolate(B=Add(A=Add(A=@n25, B=@n27), B=Multiply(A=@n28, B=@n20)), Alpha=@n22) {constAlpha=1}
  n29 = Constant3Vector() {constant=<rgba 0 0 1 0>}
  n30 = Time()
  n31 = ScalarParameter() {defaultValue=-0.2, group=MeshRotation, parameterName=RotationSpeed}
  StaticSwitchParameter(
   True=RotateAboutAxis(
    NormalizedRotationAxis=Normalize(VectorInput=@n29)
    RotationAngle=Multiply(A=@n30, B=@n31)
    PivotPoint=MaterialFunctionCall() {materialFunction=&ObjectPivotPoint}.Object Pivot Location
    Position=WorldPosition().XYZ
   )
   False=Constant()
  ) {parameterName=Is Hologram}
 ]
}
