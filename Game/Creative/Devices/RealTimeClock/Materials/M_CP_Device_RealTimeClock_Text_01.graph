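# Unlit translucent material for the Real Time Clock device's text, with an Is Hologram switch and an Enabled parameter that fades color and opacity.
# The text texture's red channel blends Glow Color into Color, and also drives opacity in hologram mode.
# Otherwise opacity comes from the green channel, masked by a biased custom-depth comparison that mobile skips.
# Hologram mode also rotates vertices over time around the object pivot's Z axis.
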Material /Game/Creative/Devices/RealTimeClock/Materials/M_CP_Device_RealTimeClock_Text_01
graph {
 properties {blendMode=BLEND_Translucent, pixelDepthOffsetMode=PDOM_Legacy, shadingModel=MSM_Unlit, usedShadingModels=Unlit}
 nodes [
  n1 = TextureSampleParameter2D(UVs=TextureCoordinate()) {parameterName=Text Texture, texture=/Game/Creative/Devices/RealTimeClock/Textures/T_CP_Device_RealTimeClock_TextDefault_01_D}
  n2 = LinearInterpolate(A=@n1.G, B=Constant() {r=1}, Alpha=@n1.G) {constAlpha=1}
  n3 = PixelDepth()
  n4 = ScalarParameter() {defaultValue=-0.25, parameterName=BiasSorting_Far}
  n5 = SceneTexture() {sceneTextureId=PPI_CustomDepth}
  n6 = Constant()
  n7 = ShadingPathSwitch(
   Default=If(A=Add(A=@n3.R, B=@n4), B=ComponentMask(None=@n5.Color) {r=true}, "A > B"=@n6, "A == B"=@n6, "A < B"=Constant() {r=1})
   Mobile=Constant() {r=1}
  ) {desc=Migrated from FeatureLevelSwitch}
  n8 = ScalarParameter() {defaultValue=1, parameterName=Enabled}
  LinearInterpolate(B=StaticSwitchParameter(True=@n1.R, False=Multiply(A=@n2, B=@n7)) {parameterName=Is Hologram}, Alpha=@n8) {constAlpha=1}
  LinearInterpolate(
   B=LinearInterpolate(
    A=VectorParameter() {parameterName=Glow Color, sortPriority=1, defaultValue=<rgba 0 1 0.9387 1>}.RGB
    B=VectorParameter() {parameterName=Color, sortPriority=1, defaultValue=<rgba 0.9047 1 1 1>}.RGB
    Alpha=@n1.R
   ) {constAlpha=1}
   Alpha=@n8
  ) {constAlpha=1}
  n9 = Constant3Vector() {constant=<rgba 0 0 1 0>}
  n10 = Time()
  n11 = ScalarParameter() {defaultValue=-0.2, group=MeshRotation, parameterName=RotationSpeed}
  StaticSwitchParameter(
   True=RotateAboutAxis(
    NormalizedRotationAxis=Normalize(VectorInput=@n9)
    RotationAngle=Multiply(A=@n10, B=@n11)
    PivotPoint=MaterialFunctionCall() {materialFunction=&ObjectPivotPoint}.Object Pivot Location
    Position=WorldPosition().XYZ
   )
   False=Constant()
  ) {parameterName=Is Hologram}
 ]
}
